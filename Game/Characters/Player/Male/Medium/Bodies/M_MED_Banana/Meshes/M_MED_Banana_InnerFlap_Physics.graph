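# Physics asset for the inner peel flaps of the medium male Banana player body, covering the dyn_flap bones and spine_05.
# Each bone gets one box body, forming a three-bone left chain and a two-bone right chain that both attach to spine_05.
# Flap segments are constrained with tighter swing and twist limits, while both base-to-spine joints leave swing and twist free.
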PhysicsAsset /Game/Characters/Player/Male/Medium/Bodies/M_MED_Banana/Meshes/M_MED_Banana_InnerFlap_Physics
bodies {
 dyn_flapBase_r [{shapeType=Box, extentX=4.6459, extentY=8.7066, extentZ=15.9202, center=<3.8919 -1.3633 -1.2153>, rotation=<rot 8.0256 -90 20.0178>}]
 dyn_flap_01_l [{shapeType=Box, extentX=4.6459, extentY=8.7066, extentZ=15.9202, center=<3.4123 0.0003 0>, rotation=<rot 0 90 0>}]
 dyn_flap_02_l [{shapeType=Box, extentX=4.6552, extentY=8.7301, extentZ=9.2716, center=<4.1931 0 0>, rotation=<rot 0 90 0>}]
 dyn_flap_Base_l [{shapeType=Box, extentX=2.302, extentY=2.1483, extentZ=2.1474, center=<-1.9699 0 -0.0001>}]
 dyn_flap_r [{shapeType=Box, extentX=5.052, extentY=16.1483, extentZ=12.6474, center=<7.1406 -0.0001 0>, rotation=<rot 0 90 0>}]
 spine_05 [{shapeType=Box, extentX=15.303, extentY=26.8611, extentZ=24.2002, center=<26.8146 1.873 0>, rotation=<rot 0 21.0346 0>}]
}
constraints [
 {bone1Name=dyn_flap_01_l, bone2Name=dyn_flap_Base_l, swing1LimitDegrees=45, swing2LimitDegrees=10, twistLimitDegrees=12}
 {bone1Name=dyn_flap_02_l, bone2Name=dyn_flap_01_l, swing1LimitDegrees=45, swing2LimitDegrees=10, twistLimitDegrees=12}
 {bone1Name=dyn_flap_r, bone2Name=dyn_flapBase_r, swing1LimitDegrees=45, swing2LimitDegrees=10, twistLimitDegrees=3}
 {
  bone1Name=dyn_flap_Base_l
  bone2Name=spine_05
  swing1LimitDegrees=45
  swing1Motion=Free
  swing2LimitDegrees=45
  swing2Motion=Free
  twistLimitDegrees=45
  twistMotion=Free
 }
 {
  bone1Name=dyn_flapBase_r
  bone2Name=spine_05
  swing1LimitDegrees=45
  swing1Motion=Free
  swing2LimitDegrees=45
  swing2Motion=Free
  twistLimitDegrees=45
  twistMotion=Free
 }
]
