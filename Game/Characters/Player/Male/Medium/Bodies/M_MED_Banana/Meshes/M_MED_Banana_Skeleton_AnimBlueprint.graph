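# Animation Blueprint for the Banana player body that copies the attached parent mesh's pose and adds leg and peel-flap secondary motion.
# Leg-up alphas shift the thigh twist bones, then the cached pose feeds constraints tying both flap bases to spine and neck.
# Two RigidBody simulations with separate physics assets follow; an ADS alpha weights the inner-flap one.
# A final ADS-weighted bone modifier rotates the right flap base, with interpolated blending.
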AnimBlueprint /Game/Characters/Player/Male/Medium/Bodies/M_MED_Banana/Meshes/M_MED_Banana_Skeleton_AnimBlueprint
target_skeleton=&/M_MED_Banana_Skeleton
graphs [
 AnimGraph {nodes [
   n1 = "Constraint - dyn_flapBase_r (PP)"!(
    ComponentPose=Local To Component(LocalPose=Use cached pose 'PostLegShapes'())
    Alpha=1
    ConstraintWeights_0=0.8
    ConstraintWeights_1:float=0.2
   ) -> (Pose) {
    alphaBoolBlend {blendOption=Linear}
    boneToModify {boneName=dyn_flapBase_r}
    constraintSetup [{targetBone {boneName=spine_05}}, {targetBone {boneName=neck_01}}]
    constraintWeights [1,1]
   }
   n2 = "Constraint - dyn_flap_Base_l (PP)"!(ComponentPose=@n1, Alpha=1, ConstraintWeights_0=0.8, ConstraintWeights_1:float=0.2) -> (Pose) {
    alphaBoolBlend {blendOption=Linear}
    boneToModify {boneName=dyn_flap_Base_l}
    constraintSetup [{targetBone {boneName=spine_05}}, {targetBone {boneName=neck_01}}]
    constraintWeights [1,1]
   }
   n3 = RigidBody(ComponentPose=@n2) {
    bClampLinearTranslationLimitToRefPose=true
    bOverrideWorldGravity=true
    overridePhysicsAsset=&/M_MED_Banana_Physics
    simulationSpace=BaseBoneSpace
    alphaBoolBlend {blendOption=Linear}
    baseBoneRef {boneName=root}
    componentAppliedLinearAccClamp=<800 800 800>
    componentLinearAccScale=<0.65 0.65 0.65>
    componentLinearVelScale=<0.3 0.3 0.3>
    overrideWorldGravity {z=-980}
   }
   n4 = RigidBody(ComponentPose=@n3, Alpha=Get ADSRigidBodyAlpha() -> (ADSRigidBodyAlpha:float)) {
    bClampLinearTranslationLimitToRefPose=true
    bOverrideWorldGravity=true
    overridePhysicsAsset=&/M_MED_Banana_InnerFlap_Physics
    simulationSpace=BaseBoneSpace
    alphaBoolBlend {blendOption=Linear}
    alphaScaleBiasClamp {bInterpResult=true, interpSpeedDecreasing=2, interpSpeedIncreasing=2}
    baseBoneRef {boneName=root}
    componentAppliedLinearAccClamp=<800 800 800>
    componentLinearAccScale=<0.65 0.65 0.65>
    componentLinearVelScale=<0.3 0.3 0.3>
    overrideWorldGravity {z=-980}
   }
   n5 = "Transform (Modify) Bone - Bone: dyn_flapBase_r"(ComponentPose=@n4, Alpha=Get ADSFlapAlpha() -> (ADSFlapAlpha:float), Translation="0,0,0", Rotation="0,-20,0", Scale="1,1,1") {
    rotationMode=BMM_Replace
    rotationSpace=BCS_BoneSpace
    alphaBoolBlend {blendOption=Linear}
    alphaScaleBiasClamp {bInterpResult=true, interpSpeedDecreasing=3.5, interpSpeedIncreasing=8}
    boneToModify {boneName=dyn_flapBase_r}
    rotation {yaw=-20}
   }
   Output Pose(Result=Component To Local(ComponentPose=@n5)) {layerGroup=None}
   n6 = "Transform (Modify) Bone - Bone: thigh_twist_shift_l"(
    ComponentPose=Local To Component(LocalPose=Copy Pose From Mesh() {bCopyCurves=true, bUseAttachedParent=true})
    Alpha=Get LeftLegUpAlpha() -> (LeftLegUpAlpha:float)
    Translation="15,0,0"
    Rotation="0,0,0"
    Scale="1,1,1"
   ) {
    translationMode=BMM_Additive
    translationSpace=BCS_BoneSpace
    alphaBoolBlend {blendOption=Linear}
    boneToModify {boneName=thigh_twist_shift_l}
    translation {x=15}
   }
   n7 = "Transform (Modify) Bone - Bone: thigh_twist_shift_r"(ComponentPose=@n6, Alpha=Get RightLegUpAlpha() -> (RightLegUpAlpha:float), Translation="-15,0,0", Rotation="0,0,0", Scale="1,1,1") {
    translationMode=BMM_Additive
    translationSpace=BCS_BoneSpace
    alphaBoolBlend {blendOption=Linear}
    boneToModify {boneName=thigh_twist_shift_r}
    translation {x=-15}
   }
   Save cached pose 'PostLegShapes'(Pose=Component To Local(ComponentPose=@n7)) {cacheName=PostLegShapes}
  ]}
 EventGraph {links [], nodes []}
]
variables []
